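# Translucent material for the Reverb wheel cosmetic effect, drawing animated pulsing rings in a tintable emissive color.
# A radial mask from tiled second-channel UVs, shaped by size and sharpness parameters, indexes a packed curve texture.
# A time-driven sine of that curve, weighted by the curve and vertex alpha, becomes opacity.
# The resulting material attributes pass through the shared vehicle damage and de-rez transition functions.
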Material /VehicleCosmetics/Wheels/Reverb/Materials/M_Effect_Reverb
graph {
 properties {
  bAutomaticallySetUsageInEditor=false
  bUseMaterialAttributes=true
  bUsedWithMeshDeformer=true
  bUsedWithSkeletalMesh=true
  blendMode=BLEND_Translucent
  usedShadingModels=Default Lit
 }
 nodes [
  n1 = Multiply(A=Constant() {r=6}, B=VectorParameter() {parameterName=Custom Color, defaultValue=<rgba 1 1 1 1>}.RGB)
  n2 = Add(A=Frac(None=TextureCoordinate() {coordinateIndex=1}), B=Constant() {r=-0.5})
  n3 = DotProduct(A=@n2, B=@n2)
  n4 = ScalarParameter() {defaultValue=1, parameterName=Circle Sharpness}
  n5 = Multiply(A=Subtract(A=OneMinus(None=@n3), B=ScalarParameter() {defaultValue=0.8, parameterName=Circle Size}), B=@n4)
  n6 = TextureSampleParameter2D(UVs=SquareRoot(None=Clamp(None=@n5))) {parameterName=Curve Generic, samplerType=SAMPLERTYPE_LinearColor, texture=/VehicleCosmetics/Wheels/Reverb/Textures/Curve_Generic_01_Pack}
  n7 = Clamp(None=@n6.RGB)
  n8 = Add(A=Multiply(A=Constant() {r=0.25}, B=Time()), B=@n7)
  n9 = Multiply(A=Multiply(A=Sine(None=@n8) {period=0.1}, B=@n7), B=@n4)
  n10 = Power(Base=@n6.RGB, Exp=@n4)
  n11 = Multiply(A=Clamp(None=Multiply(A=@n9, B=@n10)), B=VertexColor().A)
  MaterialFunctionCall(In=MaterialFunctionCall(Material=SetMaterialAttributes!(MaterialAttributes, Emissive Color=@n1, Opacity=@n11) {attributeSetTypes [#1, #2]}) {materialFunction=&MF_BaseVehicle_Damage}.Result) {materialFunction=&MF_VehicleDeRez_Transition}
 ]
}
