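# Masked parent material for the Reverb vehicle wheel, adding animated, glowing circular pulse effects to the MF_WheelParent_Simple base.
# Doubled red vertex color blends the base emissive toward a boosted Custom Color and the opacity mask toward the pulse pattern.
# The pattern shapes tiled UV1 radial gradients through a curve texture and time-driven sine, faded by vertex alpha.
# The resulting attributes pass through vehicle damage, then de-rez transition functions.
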Material /VehicleCosmetics/Wheels/Reverb/Materials/M_WheelParent_Reverb
graph {
 properties {
  bAutomaticallySetUsageInEditor=false
  bUseMaterialAttributes=true
  bUsedWithMeshDeformer=true
  bUsedWithSkeletalMesh=true
  blendMode=BLEND_Masked
  pixelDepthOffsetMode=PDOM_Legacy
  usedShadingModels=Default Lit
 }
 nodes [
  Constant()
  n1 = GetMaterialAttributes(None=MaterialFunctionCall() {materialFunction=&MF_WheelParent_Simple}.MaterialAttributes) {attributeGetTypes [#1, #2]}
  n2 = VertexColor()
  n3 = Add(A=@n2.R, B=@n2.R)
  n4 = LinearInterpolate(A=@n1.Emissive Color, B=NamedRerouteUsage(), Alpha=@n3)
  n5 = LinearInterpolate(A=Constant() {r=1}, B=NamedRerouteUsage(), Alpha=@n3)
  MaterialFunctionCall(In=MaterialFunctionCall(Material=SetMaterialAttributes!(MaterialAttributes=@n1.MaterialAttributes, Emissive Color=@n4, Opacity Mask=@n5) {attributeSetTypes [#1, #2]}) {materialFunction=&MF_BaseVehicle_Damage}.Result) {materialFunction=&MF_VehicleDeRez_Transition}
  NamedRerouteDeclaration(None=Multiply(A=Constant() {r=6}, B=VectorParameter() {group=5_Special Effects, parameterName=Custom Color, defaultValue=<rgba 1 1 1 1>}.RGB)) {name=Emissive, nodeColor=<rgba 1 0.75 0 1>}
  n6 = Constant() {r=0.25}
  n7 = Time()
  n8 = TextureCoordinate() {coordinateIndex=1}
  n9 = AppendVector(
   A=ScalarParameter() {defaultValue=1, group=5_Special Effects, parameterName=V}
   B=ScalarParameter() {defaultValue=1, group=5_Special Effects, parameterName=U}
  )
  n10 = Add(A=Frac(None=Multiply(A=@n8, B=@n9)), B=Constant() {r=-0.5})
  n11 = DotProduct(A=@n10, B=@n10)
  n12 = ScalarParameter() {defaultValue=1, group=5_Special Effects, parameterName=Circle Sharpness}
  n13 = Multiply(A=Subtract(A=OneMinus(None=@n11), B=ScalarParameter() {defaultValue=0.8, group=5_Special Effects, parameterName=Circle Size}), B=@n12)
  n14 = TextureSampleParameter2D(UVs=SquareRoot(None=Clamp(None=@n13))) {
   parameterName=Curve Generic
   samplerSource=SSM_Wrap_WorldGroupSettings
   samplerType=SAMPLERTYPE_LinearColor
   texture=/VehicleCosmetics/Wheels/Reverb/Textures/Curve_Generic_01_Pack
  }
  n15 = Clamp(None=@n14.RGB)
  n16 = Sine(None=Add(A=Multiply(A=@n6, B=@n7), B=@n15)) {period=0.1}
  n17 = Multiply(A=Multiply(A=Multiply(A=@n16, B=@n15), B=@n12), B=Power(Base=@n14.RGB, Exp=@n12))
  NamedRerouteDeclaration(None=Multiply(A=Clamp(None=@n17), B=VertexColor().A)) {name=Opacity, nodeColor=<rgba 0 1 1 1>}
 ]
}
